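# Niagara one-shot default impact effect for the Space Wanderer Male pickaxe.
# A short SubUV flash sprite and a burst of color-shifting SubUV fire sprites form the core, with panning-material mesh spike streaks around them.
# Two mesh crystal-shard bursts fly outward under curl noise, blending HDR cyan and magenta by spawn group; the smaller burst stretches with speed.
# Flash, streaks and smaller shard burst use distance culling.
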NiagaraSystem /BRCosmetics/Weapons/FORT_Melee/Pickaxe_Space_Wanderer_Male/VFX/Niagara/Impacts/NS_SpaceWandererMale_Impact_Default
dependencies {
 usedDataInterfaces [&NiagaraDataInterfaceVectorCurve, &NiagaraDataInterfaceCurve, &NiagaraDataInterfaceColorCurve]
 usedDynamicInputs []
 usedModules [/Niagara/Modules/System/SystemLifeCycle, &SystemState]
}
emitters {
 Fire {
  emitterUpdateScript [
   &EmitterState()
   &SpawnBurst_Instantaneous(Spawn Count=12, Use Spawn Probability=false)
  ]
  particleSpawnScript [
   /Niagara/Modules/Spawn/Initialization/InitializeParticle(
    Color=<rgba 1 1 1 0>
    Lifetime=&UniformRangedFloat(Minimum=0, Maximum=0.5)
    Module.Write Color=false
    Module.Write Lifetime=true
    Module.Write Mass=false
    Module.Write Material Random=true
    Module.Write Position=false
    Module.Write Scale=false
    Module.Write SpriteRotation=true
    Module.Write SpriteSize=true
    Position=&SimulationPosition()
    Sprite Rotation=&UniformRangedFloat(Minimum=-180, Maximum=180)
    Sprite Size=&Vector2DFromFloat(Value=&UniformRangedFloat(Minimum=40, Maximum=60))
   )
   &DynamicMaterialParameters(
    Index 0 Param 2=&UniformRangedFloat(Minimum=0, Maximum=1)
    Index 0 Param 3=1
    Index 0 Param 4=&UniformRangedFloat(Minimum=1, Maximum=2)
    Module.Param0WriteEnabled=false
    Module.Param1WriteEnabled=true
    Module.Param2WriteEnabled=false
    Module.Param3WriteEnabled=true
   ) {module=DynamicMaterialParameters002}
   /Niagara/Modules/Spawn/Location/SphereLocation(
    MaskSpawnGroup=false
    Module.Override Randomness=false
    Sphere Radius=&UniformRangedFloat(Minimum=0, Maximum=2)
    Surface Only=true
   )
   &CameraOffset(Camera Offset Amount=25)
   &AddVelocityFromPoint(Module.Use Falloff Distance=false)
  ]
  particleUpdateScript [
   &ParticleState()
   &SolveForcesAndVelocity(Clamp Velocity=false, Limit Acceleration=false)
   /Niagara/Modules/Update/SubUV/SubUVAnimation(Number Of Frames=32)
   &ScaleSpriteSize(Scale Factor=&Vector2DFromFloat(Value=&FloatFromCurve(FloatCurve=DataInterface() {propertyValues {Curve {keys [
          {arriveTangent=9.9004, interpMode=RCIM_Cubic, leaveTangent=1.3466, tangentMode=RCTM_User, time=0, value=0.5}
          {interpMode=RCIM_Cubic, leaveTangent=1.3, time=1, value=0.99}
         ]}}})))
   &Color(
    Color=&ColorFromCurve(ColorCurve=DataInterface() {propertyValues {
       AlphaCurve {keys [<at 0.5989 1 RCIM_Cubic>]}
       BlueCurve {keys [<at 0.2346 5 RCIM_Cubic>,<at 1 3.6888 RCIM_Cubic>]}
       GreenCurve {keys [<at 0.2346 1.1022 RCIM_Cubic>,<at 1 0 RCIM_Cubic>]}
       RedCurve {keys [<at 0.2346 0 RCIM_Cubic>,<at 1 5 RCIM_Cubic>]}
      }})
    Scale Alpha=&FloatFromCurve(
     FloatCurve=DataInterface() {propertyValues {Curve {keys [
         {arriveTangent=-0.5972, interpMode=RCIM_Cubic, leaveTangent=-0.5972, tangentMode=RCTM_User, time=0, value=0.5}
         {arriveTangent=-0.0593, interpMode=RCIM_Cubic, leaveTangent=-0.0593, tangentMode=RCTM_User, time=1, value=0}
        ]}}}
     Scale Curve=$Particles.Initial.MaterialRandom
    )
   )
   &DynamicMaterialParameters(
    Index 0 Param 1=&FloatFromCurve(FloatCurve=DataInterface() {propertyValues {Curve {keys [{arriveTangent=3.6384, interpMode=RCIM_Cubic, time=0, value=0.5}, {interpMode=RCIM_Cubic, leaveTangent=1.0681, time=0.7, value=1}]}}})
    Index 0 Param 2=1
    Index 0 Param 3=1
    Index 0 Param 4=1
    Module.Param0WriteEnabled=true
    Module.Param1WriteEnabled=true
    Module.Param2WriteEnabled=true
    Module.Param3WriteEnabled=false
   ) {module=DynamicMaterialParameters001}
  ]
  rendererClasses [&NiagaraSpriteRendererProperties]
  renderers [{rendererClass=&NiagaraSpriteRendererProperties}]
 }
 Flash_Cheap {
  emitterUpdateScript [
   &EmitterState(
    Enable Distance Culling=true
    Max Distance Response=::KillImmediately
    Module.bUseMaxDistance=true
    Module.bUseMinDistance=false
    Scalability Mode=::NewEnumerator1
   )
   &SpawnBurst_Instantaneous(Spawn Count=1, Use Spawn Probability=false)
  ]
  particleSpawnScript [
   &InitializeParticle(
    Color=<rgba 0 0.8097 1 1>
    Lifetime=0.1
    Mesh Scale=&UniformRangedVector(Minimum=<0.2 0.1 0.1>, Maximum=<1 1 1>)
    Module.Write Color=true
    Module.Write Lifetime=true
    Module.Write Mass=true
    Module.Write Material Random=true
    Module.Write Position=true
    Module.Write Scale=false
    Module.Write SpriteRotation=true
    Module.Write SpriteSize=true
    Position=&SimulationPosition()
    Sprite Rotation=&NormalizedAngleToDegrees(NormalizedAngle=&UniformRangedFloat(Minimum=0, Maximum=360))
    Sprite Size=&Vector2DFromFloat(Value=&UniformRangedFloat(Minimum=65, Maximum=70))
   )
   &FN_PerParticleLOD(Enable SpriteSize Distance Scaling=true, Sprite Size Scale Far=<2.5 2.5>)
   /SetVariables(Particles.RandomFloat=&UniformRangedFloat(Minimum=0, Maximum=1)) {isSetParameters=true}
   &CameraOffset(Camera Offset Amount=15)
  ]
  particleUpdateScript [
   &ParticleState()
   &ScaleColor(Scale RGB=&MakeVectorFromLinearColorRGB(Color=<rgba 7 2.8407 0.4871 1>), ScaleA=true, ScaleRGB=false)
   &SubUVAnimation(Random SubUV Lookup Index:NiagaraFloat=$Particles.RandomFloat, SubUV Animation Mode=::NewEnumerator1)
   &ScaleSpriteSize(Scale Factor=&Vector2DFromFloat(Value=&FloatFromCurve(FloatCurve=DataInterface() {propertyValues {Curve {keys [
          {arriveTangent=2.9657, interpMode=RCIM_Cubic, leaveTangent=2.9657, tangentMode=RCTM_User, time=0, value=0.29}
          {arriveTangent=1.5119, interpMode=RCIM_Cubic, leaveTangent=1.5119, tangentMode=RCTM_User, time=0.15, value=0.6688}
          <at 1 1.03>
         ]}}}))) {module=ScaleSpriteSize001}
  ]
  rendererClasses [&NiagaraSpriteRendererProperties]
  renderers [{rendererClass=&NiagaraSpriteRendererProperties}]
 }
 SpikeStreakMesh_NEW {
  emitterUpdateScript [
   &EmitterState(
    Enable Distance Culling=true
    Max Distance Response=::KillImmediately
    Module.bUseMaxDistance=true
    Module.bUseMinDistance=false
    Scalability Mode=::NewEnumerator1
   )
   &SpawnBurst_Instantaneous(Spawn Count=&UniformRangedInt(Minimum=6, Maximum=9), Use Spawn Probability=false)
  ]
  particleSpawnScript [
   &InitializeParticle(
    Color=<rgba 1 0 1 0.35>
    Lifetime=&UniformRangedFloat(Minimum=0.1, Maximum=0.2)
    Mesh Scale=&Multiply_VectorByFloat(Vector=&UniformRangedVector(Minimum=<1.5 0.3 0.3>, Maximum=<2.3 0.45 0.45>), Float=0.6)
    Module.Write Color=true
    Module.Write Lifetime=true
    Module.Write Mass=true
    Module.Write Material Random=true
    Module.Write Position=true
    Module.Write Scale=true
    Module.Write SpriteRotation=false
    Module.Write SpriteSize=false
    Position=&SimulationPosition()
   )
   /Niagara/Modules/Spawn/Orientation/V2/InitialMeshOrientation(
    Orientation Vector=<0 1 1>
    Rotation=&UniformRangedVector(Minimum=<0 0 0>, Maximum=<1 1 1>)
    Use Orientation Vector=true
    Use Rotation Vector=true
   )
   &CameraOffset()
  ]
  particleUpdateScript [
   &ParticleState()
   &ScaleMeshSize(Scale Factor=&VectorFromCurve(VectorCurve=DataInterface() {propertyValues {
       bOptimizeLUT=false
       XCurve {keys [{arriveTangent=1.8635, interpMode=RCIM_Cubic, leaveTangent=1.8635, tangentMode=RCTM_User, time=-0.0048, value=0.4}, <at 0.9737 1>]}
       YCurve {keys [
         {arriveTangent=0.1613, interpMode=RCIM_Cubic, leaveTangent=0.1613, tangentMode=RCTM_User, time=0, value=1}
         {arriveTangent=-1.3776, interpMode=RCIM_Cubic, leaveTangent=-1.3776, tangentMode=RCTM_User, time=0.9944, value=0}
        ]}
       ZCurve {keys [<at 0 1 RCIM_Cubic>, {arriveTangent=-1.635, interpMode=RCIM_Cubic, leaveTangent=-1.635, tangentMode=RCTM_User, time=0.9961, value=0}]}
      }}))
   /Game/Effects/Niagara/Modules/Materials/MaterialPanner(Offset:NiagaraFloat=&Add_Float(A=$Emitter.Age, B=$Particles.Initial.MaterialRandom), Speed=-5)
  ]
  rendererClasses [&NiagaraMeshRendererProperties]
  renderers [{rendererClass=&NiagaraMeshRendererProperties, rendererIndex=0}]
 }
 crystal_shards {
  emitterUpdateScript [
   &EmitterState()
   &SpawnBurst_Instantaneous(Spawn Count=30, Use Spawn Probability=false)
  ]
  particleSpawnScript [
   /Game/Effects/Niagara/Modules/Groups/FN_SetSpawnGroup_v1/FN_SetSpawnGroup(Emitter Bone Count=2)
   &InitializeParticle(
    Color=&LerpLinearColors(
     StartColor=&MakeLinearColorFromVectorAndFloat(
      "Vector (RGB)"=&MakeVectorFromLinearColorRGB(Color=<rgba 0 80.9736 100 1>)
      "Float (Alpha)"=&UniformRangedFloat(Minimum=0, Maximum=2)
     )
     EndColor=&MakeLinearColorFromVectorAndFloat(
      "Vector (RGB)"=&MakeVectorFromLinearColorRGB(Color=<rgba 50 0 50 1>)
      "Float (Alpha)"=&UniformRangedFloat(Minimum=0, Maximum=3)
     )
     LerpFactor=HlslExpression() {hlslExpression=Particles.SpawnGroup}
    )
    Lifetime=&Lerp_Float(
     A=&UniformRangedFloat(Minimum=0.1, Maximum=0.345)
     B=&UniformRangedFloat(Minimum=0.1, Maximum=0.345)
     Alpha=HlslExpression() {hlslExpression=Particles.SpawnGroup}
    )
    Mesh Scale=&UniformRangedVector(Minimum=<0.025 0.025 0.0067>, Maximum=<0.05 0.05 0.025>)
    Module.Write Color=true
    Module.Write Lifetime=true
    Module.Write Mass=false
    Module.Write Material Random=true
    Module.Write Position=false
    Module.Write Scale=true
    Module.Write SpriteRotation=false
    Module.Write SpriteSize=false
    Sprite Size=<5 5>
   )
   &CylinderLocation(
    Cylinder Height=10
    Cylinder Radius=&Lerp_Float(
     A=&UniformRangedFloat(Minimum=0.35, Maximum=0.6)
     B=&UniformRangedFloat(Minimum=4, Maximum=7)
     Alpha=HlslExpression() {hlslExpression=Particles.SpawnGroup}
    )
    MaskSpawnGroup=false
    Module.Enable Lathe Profile=false
    Module.Override Randomness=false
    Module.Surface Only=true
    Orientation Axis:ENiagaraOrientationAxis=::YAxis
    Surface Only Band Thickness:NiagaraFloat=&Lerp_Float(A=5, B=2, Alpha=HlslExpression() {hlslExpression=Particles.SpawnGroup})
    Use Endcaps In Surface Only Mode=false
   )
   &AddVelocityFromPoint(Module.Use Falloff Distance=false, Velocity Strength=&UniformRangedFloat(Minimum=200, Maximum=200))
   &InheritVelocity(Module.Limit Speed=true, Velocity Scale=<2 2 2>)
   &CameraOffset(Camera Offset Amount=15)
  ]
  particleUpdateScript [
   &ParticleState()
   /Niagara/Modules/Update/Forces/CurlNoiseForce(Noise Frequency=5, Noise Strength=500, Pan Noise=false, Random Seed=1)
   &Drag(Drag=6, Use Linear Drag=true, Use Rotational Drag=false)
   &SolveForcesAndVelocity(Clamp Velocity=false, Limit Acceleration=false)
   &ScaleMeshSize(Scale Factor=&Lerp_Vector(
     A=&VectorFromFloat(Value=&FloatFromCurve(FloatCurve=DataInterface() {propertyValues {Curve {keys [
           <at 0 1>
           {arriveTangent=-0.5431, interpMode=RCIM_Cubic, leaveTangent=-0.5431, tangentMode=RCTM_User, time=0.3, value=1}
           {arriveTangent=-5.5854, interpMode=RCIM_Cubic, leaveTangent=-5.5854, tangentMode=RCTM_User, time=1, value=0}
          ]}}}))
     B=&VectorFromFloat(Value=&FloatFromCurve(FloatCurve=DataInterface() {propertyValues {Curve {keys [
           <at 0 1>
           {arriveTangent=-0.5431, interpMode=RCIM_Cubic, leaveTangent=-0.5431, tangentMode=RCTM_User, time=0.6, value=1}
           {arriveTangent=-5.5854, interpMode=RCIM_Cubic, leaveTangent=-5.5854, tangentMode=RCTM_User, time=1, value=0}
          ]}}}))
     Alpha=HlslExpression() {hlslExpression=Particles.SpawnGroup}
    ))
   &ScaleColor(
    Scale Alpha=&FloatFromCurve(FloatCurve=DataInterface() {propertyValues {Curve {keys [<at 0 1>,<at 1 0>]}}})
    Scale RGB=&VectorFromFloat(Value=&FloatFromCurve(FloatCurve=DataInterface() {propertyValues {Curve {keys [<at 0 1>,<at 1 0>]}}}))
    ScaleA=true
    ScaleRGB=true
   )
  ]
  rendererClasses [&NiagaraMeshRendererProperties]
  renderers [{rendererClass=&NiagaraMeshRendererProperties, rendererIndex=0}]
 }
 crystal_shards001 {
  emitterUpdateScript [
   &EmitterState(Enable Distance Culling=true, MaxDistance=1024, Module.bUseMaxDistance=true, Module.bUseMinDistance=false, Scalability Mode=::NewEnumerator1)
   &SpawnBurst_Instantaneous(Spawn Count=15, Use Spawn Probability=false)
  ]
  particleSpawnScript [
   &FN_SetSpawnGroup(Emitter Bone Count=2)
   &InitializeParticle(
    Color=&LerpLinearColors(
     StartColor=&MakeLinearColorFromVectorAndFloat(
      "Vector (RGB)"=&MakeVectorFromLinearColorRGB(Color=<rgba 0 80.9736 100 1>)
      "Float (Alpha)"=&Lerp_Float(
       A=&UniformRangedFloat(Minimum=0.8, Maximum=1.2)
       B=&UniformRangedFloat(Minimum=1.5, Maximum=3)
       Alpha=$Emitter.AccelerationRamp
      )
     )
     EndColor=<rgba 50 0 50 1>
     LerpFactor=HlslExpression() {hlslExpression=Particles.SpawnGroup}
    )
    Lifetime=&Lerp_Float(
     A=&UniformRangedFloat(Minimum=0.1, Maximum=0.3)
     B=&UniformRangedFloat(Minimum=0.1, Maximum=0.3)
     Alpha=HlslExpression() {hlslExpression=Particles.SpawnGroup}
    )
    Mesh Scale=&Multiply_Vector(A=<0.6 0.6 0.6>, B=&UniformRangedVector(Minimum=<0.025 0.01 0.01>, Maximum=<0.05 0.05 0.025>))
    Module.Write Color=true
    Module.Write Lifetime=true
    Module.Write Mass=false
    Module.Write Material Random=true
    Module.Write Position=false
    Module.Write Scale=true
    Module.Write SpriteRotation=false
    Module.Write SpriteSize=false
    Sprite Size=<5 5>
   )
   &SphereLocation(
    MaskSpawnGroup=false
    Module.Override Randomness=false
    Sphere Radius=&Lerp_Float(
     A=&UniformRangedFloat(Minimum=10, Maximum=20)
     B=&UniformRangedFloat(Minimum=20, Maximum=30)
     Alpha=HlslExpression() {hlslExpression=Particles.SpawnGroup}
    )
    Surface Only=false
   )
   &AddVelocityFromPoint(Module.Use Falloff Distance=false, Velocity Strength=&UniformRangedFloat(Minimum=100, Maximum=200))
   &CameraOffset(Camera Offset Amount=15)
  ]
  particleUpdateScript [
   &ParticleState()
   &CurlNoiseForce(Noise Frequency=5, Noise Strength=500, Pan Noise=false, Random Seed=1)
   &SolveForcesAndVelocity(Clamp Velocity=false, Limit Acceleration=false)
   &ScaleMeshSize(Scale Factor=&Lerp_Vector(
     A=&VectorFromFloat(Value=&FloatFromCurve(FloatCurve=DataInterface() {propertyValues {Curve {keys [
           <at 0 1>
           {arriveTangent=-0.5431, interpMode=RCIM_Cubic, leaveTangent=-0.5431, tangentMode=RCTM_User, time=0.3, value=1}
           {arriveTangent=-5.5854, interpMode=RCIM_Cubic, leaveTangent=-5.5854, tangentMode=RCTM_User, time=1, value=0}
          ]}}}))
     B=&VectorFromFloat(Value=&FloatFromCurve(FloatCurve=DataInterface() {propertyValues {Curve {keys [
           <at 0 1>
           {arriveTangent=-0.5431, interpMode=RCIM_Cubic, leaveTangent=-0.5431, tangentMode=RCTM_User, time=0.6, value=1}
           {arriveTangent=-5.5854, interpMode=RCIM_Cubic, leaveTangent=-5.5854, tangentMode=RCTM_User, time=1, value=0}
          ]}}}))
     Alpha=HlslExpression() {hlslExpression=Particles.SpawnGroup}
    ))
   &ScaleColor(
    Scale Alpha=&FloatFromCurve(FloatCurve=DataInterface() {propertyValues {Curve {keys [<at 0 1>,<at 1 0>]}}})
    Scale RGB=&VectorFromFloat(Value=&FloatFromCurve(FloatCurve=DataInterface() {propertyValues {Curve {keys [<at 0 1>,<at 1 0>]}}}))
    ScaleA=true
    ScaleRGB=true
   )
   &ScaleMeshSizeBySpeed(Max Scale Factor=<50 1 1>)
  ]
  rendererClasses [&NiagaraMeshRendererProperties]
  renderers [{rendererClass=&NiagaraMeshRendererProperties, rendererIndex=0}]
 }
}
summary {
 systemName=NS_SpaceWandererMale_Impact_Default
 emitters [
  {emitterName=SpikeStreakMesh_NEW, rendererClasses [&NiagaraMeshRendererProperties]}
  {emitterName=crystal_shards, rendererClasses [&NiagaraMeshRendererProperties]}
  {emitterName=crystal_shards001, rendererClasses [&NiagaraMeshRendererProperties]}
  {emitterName=Flash_Cheap, rendererClasses [&NiagaraSpriteRendererProperties]}
  {emitterName=Fire, rendererClasses [&NiagaraSpriteRendererProperties]}
 ]
 userVariables []
}
